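# Sound cue for the burst heard when a Rift Butterfly rift encounter opens.
# A mixer layers a random pick of four burst waves with a random pick of three distant burst waves.
# The distant layer has its own attenuation node with a much longer falloff than the cue's attenuation override, so it carries farther.
# The cue plays in the Fort_Events sound class with limited concurrency and virtualization disabled.
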SoundCue /Figment_S04_Geode/Sounds/Fort_GamePlay_Sounds/RiftButterfly/S04_Butterfly_RiftEncounter_Open_Cue
root=Mixer_0
nodes [
 {
  id=Attenuation_3
  children [Random_2]
  properties {bOverrideAttenuation=true, attenuationOverrides {falloffDistance=250000, reverbDistanceMax=250000, stereoSpread=250000}}
 }
 {id=Mixer_0, children [Random_0, Attenuation_3], properties {inputVolume [0.7,0.636]}}
 {
  id=Random_0
  children [WavePlayer_13, WavePlayer_14, WavePlayer_15, WavePlayer_16]
  properties {bShouldExcludeFromBranchCulling=true, weights [1,1,1,1]}
 }
 {id=Random_2, children [WavePlayer_17, WavePlayer_18, WavePlayer_19], properties {bShouldExcludeFromBranchCulling=true, weights [1,1,1]}}
 {
  id=WavePlayer_13
  properties {soundWaveAssetPtr=&/S04_RiftEncounter_Open_Burst_01}
 }
 {
  id=WavePlayer_14
  properties {soundWaveAssetPtr=&/S04_RiftEncounter_Open_Burst_02}
 }
 {
  id=WavePlayer_15
  properties {soundWaveAssetPtr=&/S04_RiftEncounter_Open_Burst_03}
 }
 {
  id=WavePlayer_16
  properties {soundWaveAssetPtr=&/S04_RiftEncounter_Open_Burst_04}
 }
 {
  id=WavePlayer_17
  properties {soundWaveAssetPtr=&/S04_RiftEncounter_Open_Burst_Distant_v2_01}
 }
 {
  id=WavePlayer_18
  properties {soundWaveAssetPtr=&/S04_RiftEncounter_Open_Burst_Distant_v2_02}
 }
 {
  id=WavePlayer_19
  properties {soundWaveAssetPtr=&/S04_RiftEncounter_Open_Burst_Distant_v2_03}
 }
]
properties {
 bOverrideAttenuation=true
 bOverrideConcurrency=true
 maxDistance=250400
 soundClassObject=&Fort_Events
 subtitlePriority=10001
 virtualizationMode=Disabled
 volumeMultiplier=1.5
 attenuationOverrides {falloffDistance=100000, stereoSpread=100000, attenuationShapeExtents {x=1000}}
 concurrencyOverrides {maxCount=3}
}
