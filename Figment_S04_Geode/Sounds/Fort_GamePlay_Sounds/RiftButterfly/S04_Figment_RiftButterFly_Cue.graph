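# Sound cue for the Figment S04 Geode rift butterfly, playing a looping butterfly wave in the Fort_Events sound class.
# An Enveloper wraps the looping wave player and ramps volume up from a low level to full at the start.
# The cue references the close-range Geode attenuation asset but has attenuation override enabled, disables virtualization, and bypasses volume scaling for priority.
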SoundCue /Figment_S04_Geode/Sounds/Fort_GamePlay_Sounds/RiftButterfly/S04_Figment_RiftButterFly_Cue
root=Enveloper_0
nodes [
 {
  id=Enveloper_0
  children [WavePlayer_0]
  properties {pitchCurve {editorCurveData {keys [{time=0}]}}, volumeCurve {editorCurveData {keys [<at 0 0.1>,<at 1 1>]}}}
 }
 {
  id=WavePlayer_0
  properties {bLooping=true, soundWaveAssetPtr=&/S04_New_ButterFly}
 }
]
properties {
 attenuationSettings=/Figment_S04_Geode/Athena/Sounds/Events/Geode/Geode_Close_Att
 bBypassVolumeScaleForPriority=true
 bOverrideAttenuation=true
 maxDistance=4000
 soundClassObject=&Fort_Events
 virtualizationMode=Disabled
 volumeMultiplier=0.608
}
