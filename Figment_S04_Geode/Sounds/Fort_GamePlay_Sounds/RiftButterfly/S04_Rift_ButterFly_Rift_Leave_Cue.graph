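# Sound cue for the Rift Butterfly rift leave sound in Figment S04 Geode, playing a portal whoosh.
# A single wave player feeds the output with a boosted volume multiplier under the Fort events sound class.
# It references a Geode close attenuation asset but overrides attenuation inline with a very long falloff and wide stereo spread.
# Virtualization is disabled.
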SoundCue /Figment_S04_Geode/Sounds/Fort_GamePlay_Sounds/RiftButterfly/S04_Rift_ButterFly_Rift_Leave_Cue
root=WavePlayer_4
nodes [{
  id=WavePlayer_4
  properties {soundWaveAssetPtr=&/S04_Rift_Butterfly_Portal_Whoosh_01}
 }]
properties {
 attenuationSettings=/Figment_S04_Geode/Athena/Sounds/Events/Geode/Geode_Close_Att
 bOverrideAttenuation=true
 maxDistance=20400
 soundClassObject=&Fort_Events
 virtualizationMode=Disabled
 volumeMultiplier=1.7748
 attenuationOverrides {falloffDistance=20000, stereoSpread=20000}
}
