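# Sound cue for the Rift Butterfly spawn sound in the Figment S04 Geode content.
# A Random node at the root picks one of three equally weighted butterfly movement waves each play.
# It routes through the Fort_Events sound class with the Geode close attenuation asset, reduced volume and virtualization disabled.
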SoundCue /Figment_S04_Geode/Sounds/Fort_GamePlay_Sounds/RiftButterfly/S04_Rift_ButterFly_Spawn_Cue
root=Random_0
nodes [
 {id=Random_0, children [WavePlayer_0,WavePlayer_1,WavePlayer_2], properties {weights [1,1,1]}}
 {id=WavePlayer_0, properties {soundWaveAssetPtr=&/S04_Butterfly_Move_01}}
 {id=WavePlayer_1, properties {soundWaveAssetPtr=&/S04_Butterfly_Move_02}}
 {id=WavePlayer_2, properties {soundWaveAssetPtr=&/S04_Butterfly_Move_03}}
]
properties {
 attenuationSettings=/Figment_S04_Geode/Athena/Sounds/Events/Geode/Geode_Close_Att
 maxDistance=20000
 soundClassObject=&Fort_Events
 virtualizationMode=Disabled
 volumeMultiplier=0.704
}
